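# Niagara skydiving contrail cosmetic with a pink heart theme, adjusted for dive state, time of day and front-end preview.
# Smoke-trail meshes, floating heart sprites and glow sprites follow skeletal mesh bones, with trails orienting along dive velocity.
# CloudHeart periodically spawns a curve-shaped heart ring below the player, pushes it upward and changes its motion at a height threshold.
# Crossing a similar threshold emits location events; Impact_Sparks supplies noisy, fading sparks.
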NiagaraSystem /BRCosmetics/Athena/Effects/Particles/Skydiving/NS_Contrail_LiftingAura/FX/NS_Contrail_LiftingAura
dependencies {
 usedDataInterfaces [
  &NiagaraDataInterfaceArrayFloat3
  &NiagaraDataInterfaceMeshRendererInfo
  &NiagaraDataInterfaceCurve
  &NiagaraDataInterfaceArrayFloat
  &NiagaraDataInterfaceColorCurve
  &NiagaraDataInterfaceSkeletalMesh
 ]
 usedDynamicInputs [
  &SelectFloatFromArray
  &SimulationPosition
  &FN_FrontEndVector3Modifier
  &self:Abs_Z_Axis
  &TransformVector
  &NormalizeVector
 ]
 usedModules [
  &self:SystemSpawnScript.SystemScriptSource.SystemScriptGraph.NiagaraNodeAssignment_6.SetVariables_#1
  &SystemState
  /Game/Effects/Niagara/Modules/Skydiving/FN_DivingAttributes/v5/FN_DivingAttributes
  &self:SystemSpawnScript.SystemScriptSource.SystemScriptGraph.NiagaraNodeAssignment_2.SetVariables_#2
  &FN_Integrator
  &self:SystemSpawnScript.SystemScriptSource.SystemScriptGraph.NiagaraNodeAssignment_1.SetVariables_#3
  &self:SystemSpawnScript.SystemScriptSource.SystemScriptGraph.NiagaraNodeAssignment_4.SetVariables_#4
  &Timeline
 ]
}
emitters {
 CloudHeart {
  emitterUpdateScript [
   &EmitterState(
    Life Cycle Mode=::NewEnumerator1
    Loop Delay=&RandomRangeFloat(Minimum=2, Maximum=4)
    Loop Duration=2
    UseLoopDelay=true
   )
   &SpawnBurst_Instantaneous(Spawn Count=32, Use Spawn Probability=false, UseLoopCountLimit=false)
  ]
  particleSpawnScript [
   &InitializeParticle(
    Color=&MakeCustomLinearColorFromBool(
     Bool=&Boolean_OrOperation(Bool1=$User.bIsFrontEnd, Bool2=$User.bIsFrontEndPreview)
     True Color=<rgba 1 0.75 0.7625 1>
     False Color=<rgba 1 0.851 0.9591 1>
    )
    Lifetime Max=3
    Lifetime Min=3
    Lifetime Mode=::NewEnumerator1
    Sprite Size Mode=::NewEnumerator1
    Uniform Sprite Size Max=120
    Uniform Sprite Size Min=60
    UsePositionOffset=false
    Write Material Random=true
   )
   &SystemLocation(MaskSpawnGroup=false, Offset=<0 0 -250>)
   &ShapeLocation(
    MaskSpawnGroup=false
    Override Seed=false
    Radius Position:NiagaraFloat=&FloatFromCurve(
     FloatCurve=DataInterface() {propertyValues {Curve {keys [
         {arriveTangent=-1.952, interpMode=RCIM_Cubic, leaveTangent=-1.952, tangentMode=RCTM_User, time=0, value=0.75}
         {arriveTangent=-0.417, arriveTangentWeight=0.2482, interpMode=RCIM_Cubic, leaveTangent=-0.417, leaveTangentWeight=0.1245, time=0.666, value=1}
         {arriveTangent=-3.1386, interpMode=RCIM_Cubic, leaveTangent=-3.1386, tangentMode=RCTM_User, time=1, value=0.333}
        ]}}}
     CurveIndex=&Abs_Float(Float:NiagaraFloat=/Niagara/DynamicInputs/Multiply/ScaleAndBiasFloat(
       Float:NiagaraFloat=&ReturnNormalizedExecIndex(Normalized Index Includes One=false)
       Scale:NiagaraFloat=2
       Bias:NiagaraFloat=-1
      ))
     Scale Curve=&FloatFromCurve(
      FloatCurve=DataInterface() {propertyValues {Curve {keys [{tangentMode=RCTM_User, time=0, value=1}, {tangentMode=RCTM_User, time=1, value=1}]}}}
      CurveIndex=&Abs_Float(Float:NiagaraFloat=&ScaleAndBiasFloat(
        Float:NiagaraFloat=&ReturnNormalizedExecIndex(Normalized Index Includes One=false)
        Scale:NiagaraFloat=2
        Bias:NiagaraFloat=-1
       ))
     )
    )
    Ring / Disc Distribution Mode:ENiagaraSphereDistributionMode=::NewEnumerator1
    Ring Radius=25
    Rotation Angle=0.25
    Rotation Axis=<0 0 1>
    Rotation Mode=::NewEnumerator1
    Shape Primitive=::NewEnumerator4
    U Position=&FloatFromCurve(
     FloatCurve=DataInterface() {propertyValues {Curve {keys [
         <at 0 0>
         <at 0.1 0.0666>
         <at 0.3 0.25>
         {arriveTangent=0.8541, time=0.5, value=0.5}
         <at 0.7 0.75>
         <at 0.9 0.9333>
         {arriveTangent=1.5135, time=1, value=1}
        ]}}}
     CurveIndex=&ReturnNormalizedExecIndex(Normalized Index Includes One=false)
    )
    UseOwnerScale=true
   )
   &AddVelocity(Constrain To Radius=215.0099, Module.Use Falloff Distance=false, Override Seed=false, Velocity Mode=::NewEnumerator1, Velocity Speed=1) {module=AddVelocity001}
   &ApplyInitialForces()
   /SetVariables(Particles.Unlit Color:LinearColor=&MakeCustomLinearColorFromBool(
     Bool=&Boolean_OrOperation(Bool1=$User.bIsFrontEnd, Bool2=$User.bIsFrontEndPreview)
     True Color=<rgba 1.2 1.2 1.2 1>
     False Color=<rgba 1 0.851 0.9591 1>
    )) {isSetParameters=true}
  ]
  particleUpdateScript [
   &ParticleState()
   /SetVariables(
    Particles.Collided:NiagaraBool=&SetBoolByFloatComparison(
     A=&MakeFloatFromVector(Vector=Linked() {linkedVariable {name=Particles.Position, type=NiagaraPosition}}, Channel=::NewEnumerator2)
     B=&Add_Float(
      A=&MakeFloatFromVector(Vector=Linked() {linkedVariable {name=Engine.Emitter.SimulationPosition, type=NiagaraPosition}}, Channel=::NewEnumerator2)
      B=&FN_FrontEndFloatModifier(
       Gameplay=&Lerp_Float(A=-20, B=-50, Alpha=$System.Dive)
       Front End=&Lerp_Float(A=-100, B=-150, Alpha=$System.Dive)
      )
     )
     Comparison Type=::NewEnumerator2
    )
    Particles.IsDiving:NiagaraBool=&SetBoolByFloatComparison(A=$User.ParaGlideLeanAlpha, B=0.2)
   ) {isSetParameters=true}
   &ScaleColor(
    Linear Color Curve=DataInterface() {propertyValues {
      AlphaCurve {keys [<at 0.0057 1 RCIM_Cubic>]}
      BlueCurve {keys [<at 0 0.6519>,<at 0.666 0.7333>]}
      GreenCurve {keys [<at 0 0.414>,<at 0.666 0.6>]}
      RedCurve {keys [<at 0 2>,<at 0.666 1>]}
     }}
    Scale Mode=::NewEnumerator2
   ) {module=ScaleColor001}
   &ScaleColor(
    Scale Alpha=&FloatFromCurve(
     FloatCurve=DataInterface() {propertyValues {Curve {keys [
         {arriveTangent=-0.0888, interpMode=RCIM_Cubic, leaveTangent=-0.0888, tangentMode=RCTM_User, time=0, value=0}
         {arriveTangent=0.12, interpMode=RCIM_Cubic, leaveTangent=0.12, tangentMode=RCTM_User, time=0.15, value=0.9847}
         {arriveTangent=-0.0405, interpMode=RCIM_Cubic, leaveTangent=-0.0405, tangentMode=RCTM_User, time=1, value=0}
        ]}}}
     Scale Curve=&FNTimeOfDayScalar(
      Night Scale=1
      Morning Scale=1
      Day Scale=&FN_FrontEndFloatModifier(Gameplay=1, Front End=1)
      Evening Scale=1
     )
    )
    Scale RGB=&VectorFromFloat(Value=&FNTimeOfDayScalar(
      Night Scale=0.8
      Morning Scale=0.8
      Day Scale=&FN_FrontEndFloatModifier(Gameplay=1, Front End=1)
      Evening Scale=0.5
     ))
    ScaleA=true
    ScaleRGB=true
   ) {module=ScaleColor003}
   /SetVariables(Particles.Unlit Scale 01:LinearColor=&MultiplyLinearColor(
     Linear Color=$Particles.Initial.Unlit Color
     Scale Factor=&MakeVector4_FromLinearColor(LinearColor=&MultiplyLinearColor(
       Linear Color=&ColorFromCurve(ColorCurve=DataInterface() {propertyValues {
          AlphaCurve {keys [<at 0.0057 1 RCIM_Cubic>]}
          BlueCurve {keys [<at 0 0.6519>,<at 0.666 0.7333>]}
          GreenCurve {keys [<at 0 0.414>,<at 0.666 0.6>]}
          RedCurve {keys [<at 0 2>,<at 0.666 1>]}
         }})
       Scale Factor=&MakeVector4(
        X=&FNTimeOfDayScalar(
         Night Scale=0.8
         Morning Scale=0.8
         Day Scale=&FN_FrontEndFloatModifier(Gameplay=1.2, Front End=0.85)
         Evening Scale=0.5
        )
        Y=&FNTimeOfDayScalar(
         Night Scale=0.8
         Morning Scale=0.8
         Day Scale=&FN_FrontEndFloatModifier(Gameplay=1.2, Front End=0.85)
         Evening Scale=0.5
        )
        Z=&FNTimeOfDayScalar(
         Night Scale=0.8
         Morning Scale=0.8
         Day Scale=&FN_FrontEndFloatModifier(Gameplay=1.2, Front End=0.85)
         Evening Scale=0.5
        )
        W=&FloatFromCurve(
         FloatCurve=DataInterface() {propertyValues {Curve {keys [
             {arriveTangent=-0.0888, interpMode=RCIM_Cubic, leaveTangent=-0.0888, tangentMode=RCTM_User, time=0, value=0}
             {arriveTangent=0.12, interpMode=RCIM_Cubic, leaveTangent=0.12, tangentMode=RCTM_User, time=0.15, value=1}
             {arriveTangent=-0.0405, interpMode=RCIM_Cubic, leaveTangent=-0.0405, tangentMode=RCTM_User, time=1, value=0}
            ]}}}
         Scale Curve=&FNTimeOfDayScalar(
          Night Scale=1
          Morning Scale=1
          Day Scale=&FN_FrontEndFloatModifier(Gameplay=1, Front End=0.65)
          Evening Scale=1
         )
        )
       )
      ))
    )) {enabled=false, isSetParameters=true, module=SetVariables_#5}
   /SetVariables(Particles.Unlit Scale 01:LinearColor=&MultiplyLinearColor(
     Linear Color=$Particles.Initial.Unlit Color
     Scale Factor=&MakeVector4_FromLinearColor(LinearColor=&MakeCustomLinearColorFromBool(
       Bool=&Boolean_OrOperation(Bool1=$User.bIsFrontEnd, Bool2=$User.bIsFrontEndPreview)
       True Color=&MultiplyLinearColor(
        Linear Color=&ColorFromCurve(ColorCurve=DataInterface() {propertyValues {
           AlphaCurve {keys [<at 0.0057 1 RCIM_Cubic>]}
           BlueCurve {keys [<at 0 0.3259>,<at 0.666 0.7333>]}
           GreenCurve {keys [<at 0 0.207>,<at 0.666 0.6>]}
           RedCurve {keys [<at 0 1>,<at 0.666 1>]}
          }})
        Scale Factor=&MakeVector4(
         X=&FNTimeOfDayScalar(
          Night Scale=0.8
          Morning Scale=0.8
          Day Scale=&FN_FrontEndFloatModifier(Gameplay=1.2, Front End=0.85)
          Evening Scale=0.5
         )
         Y=&FNTimeOfDayScalar(
          Night Scale=0.8
          Morning Scale=0.8
          Day Scale=&FN_FrontEndFloatModifier(Gameplay=1.2, Front End=0.85)
          Evening Scale=0.5
         )
         Z=&FNTimeOfDayScalar(
          Night Scale=0.8
          Morning Scale=0.8
          Day Scale=&FN_FrontEndFloatModifier(Gameplay=1.2, Front End=0.85)
          Evening Scale=0.5
         )
         W=&FloatFromCurve(
          FloatCurve=DataInterface() {propertyValues {Curve {keys [
              {arriveTangent=-0.0888, interpMode=RCIM_Cubic, leaveTangent=-0.0888, tangentMode=RCTM_User, time=0, value=0}
              {arriveTangent=0.12, interpMode=RCIM_Cubic, leaveTangent=0.12, tangentMode=RCTM_User, time=0.15, value=1}
              {arriveTangent=-0.0405, interpMode=RCIM_Cubic, leaveTangent=-0.0405, tangentMode=RCTM_User, time=1, value=0}
             ]}}}
          Scale Curve=&FNTimeOfDayScalar(
           Night Scale=1
           Morning Scale=1
           Day Scale=&FN_FrontEndFloatModifier(Gameplay=1, Front End=0.65)
           Evening Scale=1
          )
         )
        )
       )
       False Color=&MultiplyLinearColor(
        Linear Color=&ColorFromCurve(ColorCurve=DataInterface() {propertyValues {
           AlphaCurve {keys [<at 0.0057 1 RCIM_Cubic>]}
           BlueCurve {keys [<at 0 0.8149>,<at 0.666 0.7333>]}
           GreenCurve {keys [<at 0 0.5175>,<at 0.666 0.6>]}
           RedCurve {keys [<at 0 2.5>,<at 0.666 1>]}
          }})
        Scale Factor=&MakeVector4(
         X=&FNTimeOfDayScalar(
          Night Scale=0.8
          Morning Scale=0.8
          Day Scale=&FN_FrontEndFloatModifier(Gameplay=1.2, Front End=0.85)
          Evening Scale=0.5
         )
         Y=&FNTimeOfDayScalar(
          Night Scale=0.8
          Morning Scale=0.8
          Day Scale=&FN_FrontEndFloatModifier(Gameplay=1.2, Front End=0.85)
          Evening Scale=0.5
         )
         Z=&FNTimeOfDayScalar(
          Night Scale=0.8
          Morning Scale=0.8
          Day Scale=&FN_FrontEndFloatModifier(Gameplay=1.2, Front End=0.85)
          Evening Scale=0.5
         )
         W=&FloatFromCurve(
          FloatCurve=DataInterface() {propertyValues {Curve {keys [
              {arriveTangent=-0.0888, interpMode=RCIM_Cubic, leaveTangent=-0.0888, tangentMode=RCTM_User, time=0, value=0}
              {arriveTangent=0.12, interpMode=RCIM_Cubic, leaveTangent=0.12, tangentMode=RCTM_User, time=0.15, value=1}
              {arriveTangent=-0.0405, interpMode=RCIM_Cubic, leaveTangent=-0.0405, tangentMode=RCTM_User, time=1, value=0}
             ]}}}
          Scale Curve=&FNTimeOfDayScalar(
           Night Scale=1
           Morning Scale=1
           Day Scale=&FN_FrontEndFloatModifier(Gameplay=1, Front End=0.65)
           Evening Scale=1
          )
         )
        )
       )
      ))
    )) {isSetParameters=true}
   &ScaleVelocity(Velocity Scale=&MakeCustomVectorFromBool(
     Bool=$Particles.Collided
     True Vector=&MakeVector(
      X=&FloatFromCurve(
       FloatCurve=DataInterface() {propertyValues {Curve {keys [
           {arriveTangent=-7892.6802, interpMode=RCIM_Cubic, leaveTangent=-7892.6802, tangentMode=RCTM_User, time=0, value=4000}
           {arriveTangent=-4000, interpMode=RCIM_Cubic, leaveTangent=-4000, time=0.5, value=1000}
           {interpMode=RCIM_Cubic, leaveTangent=-10000, time=1, value=0}
          ]}}}
       Scale Curve=&FN_FrontEndFloatModifier(Gameplay=1.5, Front End=&Lerp_Float(A=1.4, B=1.3, Alpha=$System.Dive))
      )
      Y=&FloatFromCurve(
       FloatCurve=DataInterface() {propertyValues {Curve {keys [
           {arriveTangent=-7892.6802, interpMode=RCIM_Cubic, leaveTangent=-7892.6802, tangentMode=RCTM_User, time=0, value=4000}
           {arriveTangent=-4000, interpMode=RCIM_Cubic, leaveTangent=-4000, time=0.5, value=1000}
           {interpMode=RCIM_Cubic, leaveTangent=-10000, time=1, value=0}
          ]}}}
       Scale Curve=&FN_FrontEndFloatModifier(Gameplay=1.5, Front End=&Lerp_Float(A=1.4, B=1.3, Alpha=$System.Dive))
      )
      Z=0
     )
     False Vector=<0 0 0>
    )) {module=ScaleVelocity005}
   &LinearForce(Force=&Lerp_Vector(A=<0 0 15000>, B=<0 0 20000>, Alpha=$System.Dive))
   &Drag(
    Drag=&MakeCustomFloatFromBool(
     Bool=$Particles.Collided
     True Float=&FloatFromCurve(FloatCurve=DataInterface() {propertyValues {Curve {keys [
          {arriveTangent=171.3061, interpMode=RCIM_Cubic, leaveTangent=171.3061, tangentMode=RCTM_User, time=0.1, value=0}
          {arriveTangent=514.8021, interpMode=RCIM_Cubic, leaveTangent=-87.0968, tangentMode=RCTM_Break, time=0.25, value=40}
          <at 1 0>
         ]}}})
     False Float=0
    )
    Use Linear Drag=true
    Use Rotational Drag=true
   ) {module=Drag001}
   &DoOnce(
    Reset:NiagaraBool=&SetBoolByFloatComparison(
     A=&MakeFloatFromVector(Vector=Linked() {linkedVariable {name=Particles.Position, type=NiagaraPosition}}, Channel=::NewEnumerator2)
     B=&Add_Float(
      A=&MakeFloatFromVector(Vector=Linked() {linkedVariable {name=Engine.Emitter.SimulationPosition, type=NiagaraPosition}}, Channel=::NewEnumerator2)
      B=&FN_FrontEndFloatModifier(
       Gameplay=&Lerp_Float(A=-55, B=-50, Alpha=$System.Dive)
       Front End=&Lerp_Float(A=-90, B=-100, Alpha=$System.Dive)
      )
     )
     Comparison Type=::NewEnumerator7
    )
    Trigger Condition:NiagaraBool=&SetBoolByFloatComparison(
     A=&MakeFloatFromVector(Vector=Linked() {linkedVariable {name=Particles.Position, type=NiagaraPosition}}, Channel=::NewEnumerator2)
     B=&Add_Float(
      A=&MakeFloatFromVector(Vector=Linked() {linkedVariable {name=Engine.Emitter.SimulationPosition, type=NiagaraPosition}}, Channel=::NewEnumerator2)
      B=&FN_FrontEndFloatModifier(
       Gameplay=&Lerp_Float(A=-55, B=-50, Alpha=$System.Dive)
       Front End=&Lerp_Float(A=-90, B=-100, Alpha=$System.Dive)
      )
     )
     Comparison Type=::NewEnumerator2
    )
   )
   &SolveForcesAndVelocity(Clamp Velocity=false, Limit Acceleration=false)
   &GenerateLocationEvent(
    Event Generation Enabled:NiagaraBool=$Particles.DoOnce.Execute
    Event Probability=1
    Event Type:ENiagara_LocEventType=::NewEnumerator2
    Use Event Probability=true
    UseEventDelay=false
   )
   &DynamicMaterialParameters(
    Index 0 Param 1=&Add_Float(
     A=&RandomRangeFloat(Minimum=0, Maximum=1)
     B=&FloatFromCurve(
      FloatCurve=DataInterface() {propertyValues {Curve {keys [
          {arriveTangent=2.7533, interpMode=RCIM_Cubic, leaveTangent=1.8958, tangentMode=RCTM_User, time=0, value=1}
          {arriveTangent=0.8292, interpMode=RCIM_Cubic, leaveTangent=0.8292, tangentMode=RCTM_User, time=0.321, value=1.3453}
          {arriveTangent=0.0529, interpMode=RCIM_Cubic, leaveTangent=0.2021, tangentMode=RCTM_User, time=1, value=1.5}
         ]}}}
      CurveIndex=&Add_Float(A=&Multiply_Float(A=1, B=0.5), B=$Particles.NormalizedAge)
     )
    )
    Index 0 Param 2=&FloatFromCurve(FloatCurve=DataInterface() {propertyValues {Curve {keys [
         {arriveTangent=0.0453, interpMode=RCIM_Cubic, tangentMode=RCTM_Break, time=0, value=-1}
         {arriveTangent=0.0604, interpMode=RCIM_Cubic, leaveTangent=0.0604, tangentMode=RCTM_User, time=1, value=-1}
        ]}}})
    Index 0 Param 3=&FloatFromCurve(FloatCurve=DataInterface() {propertyValues {Curve {keys [
         {arriveTangent=-53.9199, interpMode=RCIM_Cubic, leaveTangent=-8.9727, tangentMode=RCTM_User, time=0, value=1}
         {arriveTangent=0.2347, interpMode=RCIM_Cubic, leaveTangent=0.2723, tangentMode=RCTM_Break, time=0.25, value=0}
        ]}}})
    Index 0 Param 4=&FloatFromCurve(
     FloatCurve=DataInterface() {propertyValues {Curve {keys [<at 0.5 1>,<at 1 0>]}}}
     Scale Curve=&FNTimeOfDayScalar(
      Night Scale=2
      Morning Scale=2
      Day Scale=&FN_FrontEndFloatModifier(Gameplay=1, Front End=0.5)
      Evening Scale=3
     )
    )
    Index 2 Param 1=&FloatFromCurve(FloatCurve=DataInterface() {propertyValues {Curve {keys [
         {interpMode=RCIM_Cubic, leaveTangent=4.2485, tangentMode=RCTM_Break, time=0, value=0}
         {arriveTangent=1.0548, interpMode=RCIM_Cubic, leaveTangent=1.0548, tangentMode=RCTM_User, time=0.17, value=0.5715}
         <at 1 1>
        ]}}})
    Index 2 Param 2=&FloatFromCurve(
     FloatCurve=DataInterface() {propertyValues {Curve {keys [
         {arriveTangent=1.9056, interpMode=RCIM_Cubic, leaveTangent=1.9056, tangentMode=RCTM_User, time=0, value=0.2}
         {arriveTangent=0.0259, interpMode=RCIM_Cubic, leaveTangent=0.0259, tangentMode=RCTM_User, time=1, value=1.4}
        ]}}}
     Scale Curve=&RandomRangeFloat(Minimum=0.2, Maximum=0.5)
    )
    Index 2 Param 3=&FloatFromCurve(FloatCurve=DataInterface() {propertyValues {Curve {keys [
         {interpMode=RCIM_Cubic, leaveTangent=2.3064, tangentMode=RCTM_Break, time=0, value=1}
         {interpMode=RCIM_Cubic, leaveTangent=11.1233, tangentMode=RCTM_Break, time=1, value=2}
        ]}}})
    Index 2 Param 4=&FloatFromCurve(
     FloatCurve=DataInterface() {propertyValues {Curve {keys [
         {arriveTangent=-0.6867, interpMode=RCIM_Cubic, leaveTangent=-0.6867, tangentMode=RCTM_User, time=0, value=1}
         {interpMode=RCIM_Cubic, leaveTangent=-1.2956, tangentMode=RCTM_Break, time=1, value=0.75}
        ]}}}
     Scale Curve=1.6
    )
    Module.Param0WriteEnabled=true
    Module.Param0WriteEnabled2=true
    Module.Param1WriteEnabled=false
    Module.Param1WriteEnabled2=true
    Module.Param2WriteEnabled=false
    Module.Param2WriteEnabled2=true
    Module.Param3WriteEnabled=true
    Module.Param3WriteEnabled2=true
    Write Parameter Index 2=true
   )
  ]
  rendererClasses [&NiagaraSpriteRendererProperties]
  renderers [
   {rendererClass=&NiagaraSpriteRendererProperties}
   {rendererClass=&NiagaraSpriteRendererProperties, rendererIndex=1}
  ]
 }
 Decal {
  emitterUpdateScript [
   &EmitterState(
    Inactive Response=::NewEnumerator1
    Life Cycle Mode=::NewEnumerator1
    Loop Behavior=::NewEnumerator1
    Loop Duration Mode=::NewEnumerator1
    UseLoopDelay=false
   )
   &SpawnBurst_Instantaneous(Spawn Count=4, Use Spawn Probability=false, UseLoopCountLimit=false)
   /Game/Effects/Niagara/Modules/Spawn/SpawnRateFromMultipleSpawnGroups_2(Global Spawnrate=8) {enabled=false}
  ]
  particleSpawnScript [
   &InitializeParticle(
    Color=<rgba 1.25 0.5725 1.1371 1>
    Lifetime=0.6
    Mass Mode=::NewEnumerator1
    Sprite Size=<50 50>
    Sprite Size Mode=::NewEnumerator4
    UsePositionOffset=false
    Write Material Random=true
   ) {module=InitializeParticle001}
   &CameraOffset()
   &PartitionParticles(Number Of Partitions=4)
  ]
  particleUpdateScript [
   &ParticleState(Kill Particles When Lifetime Has Elapsed=false)
   &SkeletalMeshLocation(
    Bone / Socket Index=$Particles.PartitionParticles.Partition
    Bone Sampling Mode=::NewEnumerator4
    Inherit Velocity=false
    Interpolate Mesh Sampling=false
    MaskSpawnGroup=false
    Sampled Position Offset=&SelectVectorFromArray(
     Direct Array Index=$Particles.PartitionParticles.Partition
     Vector Selection Array=DataInterface() {propertyValues {FloatData [<10 0 0>, <-10 0 0>, <-7 5 0>, <7 -5 0>]}}
    )
    Sampled Position Offset Coordinate Space:ENiagaraMeshTransforms=::NewEnumerator3
    Skeletal Mesh=$System.Skeletal Mesh
   )
   &ScaleColor(
    Scale Alpha=&FloatFromCurve(FloatCurve=DataInterface() {propertyValues {Curve {keys [
         <at 0.0001 0 RCIM_Cubic>
         {arriveTangent=1.6347, interpMode=RCIM_Cubic, leaveTangent=1.6347, time=0.1393, value=0.6251}
         {arriveTangent=0.8334, interpMode=RCIM_Cubic, leaveTangent=0.8334, time=0.3202, value=0.5233}
         {arriveTangent=-0.4566, interpMode=RCIM_Cubic, leaveTangent=-0.4566, time=0.4613, value=0.8934}
         {arriveTangent=-0.922, interpMode=RCIM_Cubic, leaveTangent=-0.922, time=0.6241, value=0.3845}
         {arriveTangent=-1.0215, interpMode=RCIM_Cubic, leaveTangent=-1.0215, time=0.7724, value=0.6066}
         <at 1.0006 0 RCIM_Cubic>
        ]}}})
    ScaleA=true
    ScaleRGB=false
   ) {enabled=false}
   /SetVariables(Particles.CameraOffset Glow=15, Particles.Color Glow:LinearColor=<rgba 1 0.339 0.372 0.5>, Particles.Size Glow:Vector2f=<50 50>) {isSetParameters=true}
  ]
  rendererClasses [&NiagaraSpriteRendererProperties]
  renderers [
   {rendererClass=&NiagaraSpriteRendererProperties}
   {rendererClass=&NiagaraSpriteRendererProperties, rendererIndex=1}
  ]
 }
 FloatingHearts {
  emitterUpdateScript [
   &EmitterState(Delay First Loop Only=true, Life Cycle Mode=::NewEnumerator1, Loop Delay=0.2416, Loop Duration=1.2584, UseLoopDelay=true) {module=EmitterState001}
   &SpawnRate(SpawnRate=&Lerp_Float(A=8, B=16, Alpha=$System.FN_Integrator001.Value), Use Spawn Probability=true)
  ]
  particleSpawnScript [
   &PartitionParticles()
   &InitializeParticle(
    Color=&RandomRangeLinearColor(Minimum=<rgba 1 1 1 1>, Maximum=<rgba 1 0.832 0.9153 1>)
    Lifetime Min=1.6
    Lifetime Mode=::NewEnumerator1
    Sprite Size Mode=::NewEnumerator1
    Sprite UV Mode=::NewEnumerator1
    Uniform Sprite Size Max=12
    Uniform Sprite Size Min=10
    UsePositionOffset=false
    Write Material Random=true
   ) {module=InitializeParticle001}
   &ShapeLocation(
    Apply To Particle Position=::NewEnumerator1
    MaskSpawnGroup=false
    Offset=&MaskVectorBySpawnGroup(Masked Vector=<10 0 0>, Passthrough Vector=<-10 0 0>, Spawn Group Mask=$Particles.PartitionParticles.Partition)
    Offset Mode=::NewEnumerator0
    Override Seed=false
    Shape Origin=&VectorFromFloat(Value=0)
    Sphere Radius=10
    UseOwnerScale=true
   )
   /SetVariables(Particles.SubImageIndex=&RandomRangeFloat(Minimum=0, Maximum=3)) {isSetParameters=true}
   &CameraOffset(Camera Offset Amount=10) {module=CameraOffset001}
  ]
  particleUpdateScript [
   &ParticleState()
   &SkeletalMeshLocation(
    Bone / Socket Index=$Particles.PartitionParticles.Partition
    Bone Sampling Mode=::NewEnumerator4
    MaskSpawnGroup=false
    Orientation Sampling=::NewEnumerator1
    Sampled Position Offset=$StackContext.ShapeLocation.ShapePosition
    Skeletal Mesh=$System.Skeletal Mesh
   )
   /SetVariables(Particles.Position=&Add_Vector(
     A=Linked() {linkedVariable {name=Particles.Position, type=NiagaraPosition}}
     B=&Multiply_VectorByFloat(
      Vector=$System.DiveVelocityScale
      Float=&Lerp_Float(
       A=&Add_Float(
        A=&FloatFromCurve(
         FloatCurve=DataInterface() {propertyValues {Curve {keys [
             {arriveTangent=0.3747, interpMode=RCIM_Cubic, leaveTangent=0.3747, tangentMode=RCTM_User, time=0, value=0}
             {arriveTangent=1.721, interpMode=RCIM_Cubic, leaveTangent=1.721, tangentMode=RCTM_User, time=1, value=1}
            ]}}}
         Scale Curve=&Multiply_Float(A=$System.DiveOffset, B=&RandomRangeFloat(Minimum=0.5, Maximum=1))
        )
        B=-11
       )
       B=&Add_Float(
        A=&FloatFromCurve(
         FloatCurve=DataInterface() {propertyValues {Curve {keys [
             {arriveTangent=-0.0613, interpMode=RCIM_Cubic, leaveTangent=-0.0613, tangentMode=RCTM_User, time=0, value=0}
             {arriveTangent=2.6927, interpMode=RCIM_Cubic, leaveTangent=2.6927, tangentMode=RCTM_User, time=1, value=1}
            ]}}}
         Scale Curve=&Multiply_Float(A=$System.DiveOffset, B=&RandomRangeFloat(Minimum=4, Maximum=5))
        )
        B=-11
       )
       Alpha=$System.Dive
      )
     )
    )) {isSetParameters=true}
   &ScaleSpriteSize(Uniform Curve Sprite Scale=DataInterface() {propertyValues {Curve {keys [
        {arriveTangent=2.4763, tangentMode=RCTM_User, time=0.2, value=1}
        {arriveTangent=2.5123, interpMode=RCIM_Cubic, leaveTangent=2.5123, tangentMode=RCTM_User, time=0.6, value=1}
        {arriveTangent=-0.5868, interpMode=RCIM_Cubic, leaveTangent=-0.5868, tangentMode=RCTM_User, time=1, value=2}
       ]}}}) {module=ScaleSpriteSize002}
   &ScaleColor(
    Scale Alpha=&FloatFromCurve(FloatCurve=DataInterface() {propertyValues {Curve {keys [
         {interpMode=RCIM_Cubic, tangentMode=RCTM_User, time=0, value=0}
         {interpMode=RCIM_Cubic, tangentMode=RCTM_User, time=0.25, value=1}
         {tangentMode=RCTM_User, time=1, value=1}
        ]}}})
    ScaleA=true
    ScaleRGB=true
   ) {module=ScaleColor001}
   &ScaleColor(
    Scale Alpha=&FloatFromCurve(FloatCurve=DataInterface() {propertyValues {Curve {keys [
         {interpMode=RCIM_Cubic, tangentMode=RCTM_User, time=0, value=0}
         {interpMode=RCIM_Cubic, tangentMode=RCTM_User, time=0.5, value=1}
         {tangentMode=RCTM_User, time=1, value=0}
        ]}}})
    ScaleA=true
    ScaleRGB=true
   )
   &SolveForcesAndVelocity(Clamp Velocity=false, Limit Acceleration=false)
  ]
  rendererClasses [&NiagaraSpriteRendererProperties]
  renderers [{rendererClass=&NiagaraSpriteRendererProperties}]
 }
 Impact_Sparks {
  emitterUpdateScript [&EmitterState(
    Inactive Response=::NewEnumerator1
    Life Cycle Mode=::NewEnumerator1
    Loop Behavior=::NewEnumerator1
    Loop Duration Mode=::NewEnumerator1
    UseLoopDelay=false
   )]
  particleSpawnScript [&InitializeParticle(
    Color=<rgba 5 3.7043 1.51 1>
    Lifetime=&Multiply_Float(A=&RandomRangeFloat(Minimum=0.05, Maximum=1), B=1.1)
    Sprite Size Mode=::NewEnumerator1
    UsePositionOffset=false
    Write Material Random=true
   )]
  particleUpdateScript [
   &ParticleState()
   &CurlNoiseForce(Noise Strength=25, Pan Noise=false)
   &Drag(Drag=1.75, Use Linear Drag=true, Use Rotational Drag=true)
   &SolveForcesAndVelocity(Clamp Velocity=false, Limit Acceleration=false)
   &ScaleColor(
    Scale Alpha=&FloatFromCurve(FloatCurve=DataInterface() {propertyValues {Curve {keys [<at -0.0092 0.0144>,<at 0.0967 0.9767>,<at 1.0011 0.002>]}}})
    Scale RGB=<1.5 1.5 1.5>
    ScaleA=true
    ScaleRGB=true
   )
   &ScaleSpriteSize(Scale Factor=&Vector2DFromFloat(Value=&FloatFromCurve(FloatCurve=DataInterface() {propertyValues {Curve {keys [<at -0.0046 0>,<at 0.1746 0.9952>,<at 1.0011 0.002>]}}}, Scale Curve=1.5)))
  ]
  rendererClasses [&NiagaraSpriteRendererProperties]
  renderers [{rendererClass=&NiagaraSpriteRendererProperties}]
 }
 Smoke_Trails {
  emitterUpdateScript [
   &EmitterState(Life Cycle Mode=::NewEnumerator1, Loop Behavior=::NewEnumerator1, Loop Delay=0.5, Loop Duration Mode=::NewEnumerator1, UseLoopDelay=true)
   &SpawnBurst_Instantaneous(Spawn Count=4, Use Spawn Probability=false, UseLoopCountLimit=false)
   &SpawnBurst_Instantaneous(Spawn Count=2, Spawn Group=1, Use Spawn Probability=false, UseLoopCountLimit=false) {module=SpawnBurst_Instantaneous001}
   &SpawnBurst_Instantaneous(Spawn Count=2, Spawn Time=0.0547, Use Spawn Probability=false, UseLoopCountLimit=false) {module=SpawnBurst_Instantaneous002}
  ]
  particleSpawnScript [
   &PartitionParticles(Number Of Partitions=&MaskIntBySpawnGroup(Masked Int=4, Passthrough Int=2, Spawn Group Mask=0))
   &InitializeParticle(
    Color=<rgba 1 0.339 0.372 1>
    Lifetime=2
    Mesh Renderer Array Visibility Mode=::NewEnumerator1
    Mesh Renderer Info=DataInterface() {propertyValues {MeshRenderer=&self:Smoke_Trails_3.NiagaraMeshRendererProperties_0}}
    Mesh Renderer Mesh Index:NiagaraInt32=&SpawnGroup()
    Mesh Scale=&SelectVectorFromArray(
     Direct Array Index=$Particles.PartitionParticles.Partition
     Vector Selection Array=DataInterface() {propertyValues {FloatData [<1.5 1.5 0.55>, <1.5 1.5 0.55>, <1 1 0.25>, <1 1 0.25>]}}
    )
    Mesh Scale Mode=::NewEnumerator4
    Sprite Size Mode=::NewEnumerator3
    Uniform Sprite Size=50
    UsePositionOffset=false
    Write Material Random=true
   )
   &InitialMeshOrientation(
    Mesh Axis To Orient=<0 0 -1>
    Mesh Orientation Mode=::NewEnumerator2
    Rotation=&Lerp_Vector(A=<0 0 0>, B=<0.5 0 0>, Alpha=&MakeFloatFromInt(Int=$Particles.PartitionParticles.Partition))
    Rotation Coordinate Space=::NewEnumerator0
    Use Rotation Vector=true
   )
   /SetVariables(Particles.RendererTag:NiagaraInt32=&MaskIntBySpawnGroup(Masked Int=0, Passthrough Int=1, Spawn Group Mask=0)) {isSetParameters=true}
  ]
  particleUpdateScript [
   &ParticleState(
    Kill Particles When Lifetime Has Elapsed=false
    Let Infinitely Lived Particles Die When Emitter Deactivates=true
    Loop Particles Lifetime=true
   )
   &SkeletalMeshLocation(
    Bone / Socket Index=$Particles.PartitionParticles.Partition
    Bone Sampling Mode=::NewEnumerator4
    "Interpolate Mesh Sampling (Update)"=true
    MaskSpawnGroup=false
    Orientation Sampling=::NewEnumerator1
    Sampled Position Offset=&SelectVectorFromArray(
     Direct Array Index=$Particles.PartitionParticles.Partition
     Vector Selection Array=DataInterface() {propertyValues {FloatData [<10 0 0>, <-10 0 0>, <2 0 0>, <-5 0 0>]}}
    )
    Skeletal Mesh=$System.Skeletal Mesh
   )
   /SetVariables(Particles.Position=&Add_Vector(
     A=&ConvertPositionToVector(Input Position=$Particles.Position, Convert To Absolute Large World Vector=false)
     B=&Add_Vector(A=&Multiply_VectorByFloat(Vector=$System.DiveVelocityScale, Float=10), B=<0 0 0>)
    )) {isSetParameters=true}
   &UpdateMeshOrientation(
    Facing Direction=$System.DiveVelocityScale
    Facing Interpolation=false
    Orientation Method=::NewEnumerator1
    Override Quat=false
    Side Direction:Vector3f=$Engine.Owner.SystemXAxis
    Stabilization Mode:ENiagara_OrientVectorCount=::NewEnumerator2
    TransformFacing=false
    TransformSide=false
   ) {module=UpdateMeshOrientation001}
   &UpdateMeshOrientation(
    Rotation Rate=&MaskFloatBySpawnGroup(Masked Float=0, Passthrough Float=&RandomRangeFloat(Minimum=1, Maximum=1.6))
    Rotation Vector=<0 0 1>
   )
   &ScaleMeshSize(Scale Factor=&Multiply_Vector(
     A=&Lerp_Vector(A=<1.2 1.2 2>, B=<1 1 1.5>, Alpha=$System.FN_Integrator001.Value)
     B=&MakeVector(
      X=1
      Y=1
      Z=&FloatFromCurve(
       FloatCurve=DataInterface() {propertyValues {Curve {keys [<at 0 0 RCIM_Cubic>, {arriveTangent=4.1531, interpMode=RCIM_Cubic, leaveTangent=4.1531, tangentMode=RCTM_User, time=0.5, value=1}]}}}
       CurveIndex=$Emitter.Age
      )
     )
    ))
   &ScaleMeshSize(Scale Factor=&LerpPosition(
     A=&SelectVectorFromArray(
      Direct Array Index=$Particles.PartitionParticles.Partition
      Vector Selection Array=DataInterface() {propertyValues {FloatData [<1 1 1>, <1 1 1>, <1 1 1>, <1 1 1>]}}
     )
     B=&SelectVectorFromArray(
      Direct Array Index=$Particles.PartitionParticles.Partition
      Vector Selection Array=DataInterface() {propertyValues {FloatData [<1.5 1.5 1.5>, <1.5 1.5 1.5>, <1.5 1.5 2>, <1.5 1.5 2>]}}
     )
     Alpha=$System.Dive
    )) {module=ScaleMeshSize001}
   &ScaleColor(
    Scale Alpha=&FloatFromCurve(
     FloatCurve=DataInterface() {propertyValues {Curve {keys [<at 0.5 0 RCIM_Cubic>, {arriveTangent=4.1531, interpMode=RCIM_Cubic, leaveTangent=4.1531, tangentMode=RCTM_User, time=1, value=1}]}}}
     CurveIndex=$Emitter.Age
     Scale Curve=0.5
    )
    Scale RGB=&Multiply_VectorByFloat(Vector=&MakeVectorFromLinearColorRGB(Color=<rgba 1 1 1 1>), Float=0.25)
    ScaleA=true
    ScaleRGB=true
   ) {module=ScaleColor001}
   /SetVariables(Particles.SpiralColor:LinearColor=&MultiplyLinearColor(
     Linear Color=$Particles.Color
     Scale Factor=&MakeVector4_FromLinearColor(LinearColor=<rgba 0.276 0.276 0.276 0.1>)
    )) {isSetParameters=true}
   &InterpolateOverTime(
    Rate Of Change=12.5
    Target Vector Value:Vector3f=&AddVectorToPosition(Position=$StackContext.SkeletalMeshLocation.SampledPosition, Vector=<0 0 250>)
    Variable Type=::NewEnumerator2
   ) {module=InterpolateOverTime001}
   /SetVariables(Particles.TargetPosition:NiagaraPosition=$StackContext.InterpolateOverTime001.InterpolateOverTime_V3_.Moving Average) {isSetParameters=true}
   /SetVariables(Particles.TargetDelta:Vector3f=&Subtract_Vector(
     A=Linked() {linkedVariable {name=Particles.TargetPosition, type=NiagaraPosition}}
     B=Linked() {linkedVariable {name=Particles.Position, type=NiagaraPosition}}
    )) {isSetParameters=true}
   &DynamicMaterialParameters(
    Index 0 Param 1=&RandomRangeFloat(Minimum=0, Maximum=1)
    Index 0 Param 2=&RandomRangeFloat(Minimum=0, Maximum=1.5)
    Index 1 Param 1=&MakeFloatFromVector(Vector=$Particles.TargetDelta, Channel=::NewEnumerator0)
    Index 1 Param 2=&MakeFloatFromVector(Vector=$Particles.TargetDelta, Channel=::NewEnumerator1)
    Index 1 Param 3=&OneMinusFloat(Float=$System.Dive)
    Index 1 Param 4=128
    Index 2 Param 1=111
    Index 2 Param 3=0.2
    Index 2 Param 4=&MaskFloatBySpawnGroup(Masked Float=-2, Passthrough Float=1, Spawn Group Mask=1)
    Module.Param0WriteEnabled=true
    Module.Param0WriteEnabled1=true
    Module.Param0WriteEnabled2=true
    Module.Param1WriteEnabled=true
    Module.Param1WriteEnabled1=true
    Module.Param1WriteEnabled2=true
    Module.Param2WriteEnabled=false
    Module.Param2WriteEnabled1=true
    Module.Param2WriteEnabled2=true
    Module.Param3WriteEnabled=true
    Module.Param3WriteEnabled1=true
    Module.Param3WriteEnabled2=true
    Write Parameter Index 1=true
    Write Parameter Index 2=true
   )
  ]
  rendererClasses [&NiagaraMeshRendererProperties]
  renderers [
   {rendererClass=&NiagaraMeshRendererProperties, rendererIndex=0}
   {rendererClass=&NiagaraMeshRendererProperties, rendererIndex=1}
  ]
 }
}
summary {
 systemName=NS_Contrail_LiftingAura
 emitters [
  {emitterName=Smoke_Trails, rendererClasses [&NiagaraMeshRendererProperties]}
  {emitterName=FloatingHearts, rendererClasses [&NiagaraSpriteRendererProperties]}
  {emitterName=CloudHeart, rendererClasses [&NiagaraSpriteRendererProperties]}
  {emitterName=Impact_Sparks, rendererClasses [&NiagaraSpriteRendererProperties]}
  {emitterName=Decal, rendererClasses [&NiagaraSpriteRendererProperties]}
 ]
 userVariables [
  User.bIsFrontEnd {type=NiagaraBool}
  User.bIsFrontEndPreview {type=NiagaraBool}
  User.ParaGlideLeanAlpha {type=NiagaraFloat, defaultValue=0, description=""}
  User.PlayerSpeed {type=NiagaraFloat, defaultValue=0, description=""}
  User.RibbonIDOffset {type=NiagaraInt32}
 ]
}
